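# Physics asset for the SteelyGrin medium male player body cosmetic, set up for the five lower jacket bones at back, left and right.
# Each dyn jacket bone carries capsule or paired-box shapes and hangs from a tiny base anchor body with tight swing limits and zero twist.
# The anchors link to spine_01 with free angular motion, as do the pelvis, spine and thigh capsules that form the body volumes around the jacket.
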PhysicsAsset /BRCosmetics/Characters/Player/Male/Medium/Bodies/M_MED_SteelyGrin/Meshes/M_MED_SteelyGrin_Physics
bodies {
 base_low_jacket_back [{shapeType=Capsule, length=0.505, radius=0.505, shapeName=base_low_jacket_back_capsule, rotation=<rot 0 -90 0>}]
 base_low_jacket_l_01 [{shapeType=Capsule, length=0.505, radius=0.505, shapeName=base_low_jacket_l_01_capsule, rotation=<rot 0 -90 0>}]
 base_low_jacket_l_02 [{shapeType=Capsule, length=0.505, radius=0.505, shapeName=base_low_jacket_l_02_capsule, rotation=<rot 0 -90 0>}]
 base_low_jacket_r_01 [{shapeType=Capsule, length=0.505, radius=0.505, shapeName=base_low_jacket_r_01_capsule, rotation=<rot 0 -90 0>}]
 base_low_jacket_r_02 [{shapeType=Capsule, length=0.505, radius=0.505, shapeName=base_low_jacket_r_02_capsule, rotation=<rot 0 -90 0>}]
 dyn_low_jacket_back [{shapeType=Capsule, length=5.5793, radius=2.3061, shapeName=dyn_low_jacket_back_capsule, center=<4.0301 0 0>, rotation=<rot 90 0 0>}]
 dyn_low_jacket_l_01 [
  {
   shapeType=Box
   extentX=11.0353
   extentY=1.3141
   extentZ=7.97
   shapeName=dyn_low_jacket_l_01_box
   center=<3.2485 1.1031 5.0924>
   rotation=<rot 2.8231 -0.3712 7.7185>
  }
  {
   shapeType=Box
   extentX=9.8817
   extentY=1.5449
   extentZ=7.6656
   shapeName=dyn_low_jacket_l_01_box
   center=<3.2082 1.3361 -0.45>
   rotation=<rot -1.1752 0.2172 -10.4759>
  }
 ]
 dyn_low_jacket_l_02 [
  {shapeType=Capsule, length=6.8943, radius=1.8138, shapeName=dyn_low_jacket_l_02_capsule, center=<3.7056 0 0>, rotation=<rot 90 0 0>}
  {shapeType=Capsule, length=6.8943, radius=1.8138, shapeName=dyn_low_jacket_l_02_capsule, center=<3.7056 0.3315 3.6046>, rotation=<rot 90 0 0>}
 ]
 dyn_low_jacket_r_01 [
  {
   shapeType=Box
   extentX=10.2858
   extentY=1.4454
   extentZ=6.4587
   shapeName=dyn_low_jacket_l_01_box
   center=<-3.2082 -1.3361 -0.45>
   rotation=<rot -0.7147 2.7022 -10.517>
  }
  {
   shapeType=Box
   extentX=11.4394
   extentY=1.2147
   extentZ=5.1721
   shapeName=dyn_low_jacket_l_01_box
   center=<-2.5635 -1.6975 -5.9595>
   rotation=<rot 5.5903 4.0338 12.9945>
  }
 ]
 dyn_low_jacket_r_02 [{shapeType=Capsule, length=5.214, radius=2.0562, shapeName=dyn_low_jacket_r_02_capsule, center=<-3 0 0>, rotation=<rot 90 0 0>}]
 pelvis [{
   shapeType=Capsule
   length=13.3939
   radius=4.849
   shapeName=pelvis_capsule
   center=<2.9363 6.5003 0.6031>
   rotation=<rot -1.0351 -90.019 -1.1165>
  }]
 spine_01 [{
   shapeType=Capsule
   length=7.9973
   radius=9.2055
   shapeName=spine_01_capsule
   center=<-0.9561 0.7966 -0.0368>
   rotation=<rot 5.0722 -90.0004 -0.5449>
  }]
 spine_02 [{
   shapeType=Capsule
   length=7.9973
   radius=9.2055
   shapeName=spine_01_capsule
   center=<-0.9561 0.7966 -0.0368>
   rotation=<rot 5.0722 -90.0004 -0.5449>
  }]
 thigh_l [{shapeType=Capsule, length=21.8722, radius=6.6898, shapeName=thigh_l_capsule, center=<-9.7966 3.2468 0>, rotation=<rot 90 0 0>}]
 thigh_r [{shapeType=Capsule, length=18.6754, radius=6.7573, shapeName=thigh_r_capsule, center=<7.484 -2.5621 0>, rotation=<rot 90 0 0>}]
}
constraints [
 {bone1Name=dyn_low_jacket_r_02, bone2Name=base_low_jacket_r_02, swing1LimitDegrees=1, swing2LimitDegrees=1, twistLimitDegrees=0}
 {bone1Name=dyn_low_jacket_r_01, bone2Name=base_low_jacket_r_01, swing1LimitDegrees=10, swing2LimitDegrees=15, twistLimitDegrees=0}
 {bone1Name=dyn_low_jacket_l_02, bone2Name=base_low_jacket_l_02, swing1LimitDegrees=5, swing2LimitDegrees=1, twistLimitDegrees=0}
 {bone1Name=dyn_low_jacket_l_01, bone2Name=base_low_jacket_l_01, swing1LimitDegrees=15, swing2LimitDegrees=15, twistLimitDegrees=0}
 {
  bone1Name=base_low_jacket_back
  bone2Name=spine_01
  swing1LimitDegrees=45
  swing1Motion=Free
  swing2LimitDegrees=45
  swing2Motion=Free
  twistLimitDegrees=45
  twistMotion=Free
 }
 {
  bone1Name=base_low_jacket_l_01
  bone2Name=spine_01
  swing1LimitDegrees=45
  swing1Motion=Free
  swing2LimitDegrees=45
  swing2Motion=Free
  twistLimitDegrees=45
  twistMotion=Free
 }
 {
  bone1Name=base_low_jacket_l_02
  bone2Name=spine_01
  swing1LimitDegrees=45
  swing1Motion=Free
  swing2LimitDegrees=45
  swing2Motion=Free
  twistLimitDegrees=45
  twistMotion=Free
 }
 {
  bone1Name=base_low_jacket_r_01
  bone2Name=spine_01
  swing1LimitDegrees=45
  swing1Motion=Free
  swing2LimitDegrees=45
  swing2Motion=Free
  twistLimitDegrees=45
  twistMotion=Free
 }
 {
  bone1Name=base_low_jacket_r_02
  bone2Name=spine_01
  swing1LimitDegrees=45
  swing1Motion=Free
  swing2LimitDegrees=45
  swing2Motion=Free
  twistLimitDegrees=45
  twistMotion=Free
 }
 {bone1Name=dyn_low_jacket_back, bone2Name=base_low_jacket_back, swing1LimitDegrees=1, swing2LimitDegrees=1, twistLimitDegrees=0}
 {
  bone1Name=spine_02
  bone2Name=spine_01
  swing1LimitDegrees=45
  swing1Motion=Free
  swing2LimitDegrees=45
  swing2Motion=Free
  twistLimitDegrees=45
  twistMotion=Free
 }
 {
  bone1Name=spine_01
  bone2Name=pelvis
  swing1LimitDegrees=45
  swing1Motion=Free
  swing2LimitDegrees=45
  swing2Motion=Free
  twistLimitDegrees=45
  twistMotion=Free
 }
 {
  bone1Name=thigh_l
  bone2Name=pelvis
  swing1LimitDegrees=45
  swing1Motion=Free
  swing2LimitDegrees=45
  swing2Motion=Free
  twistLimitDegrees=45
  twistMotion=Free
 }
 {
  bone1Name=thigh_r
  bone2Name=pelvis
  swing1LimitDegrees=45
  swing1Motion=Free
  swing2LimitDegrees=45
  swing2Motion=Free
  twistLimitDegrees=45
  twistMotion=Free
 }
]
