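# Physics asset for the SteelyGrin male medium body's jacket collar, setting up secondary motion for its hood and jacket flaps.
# Each dyn_ hood or flap bone carries capsules and is constrained with tight swing and twist limits to a tiny base_ anchor body.
# Anchors join spine_05 with free angular constraints, as does an upper-body set of head, neck, arm and left clavicle capsules.
# With no right clavicle body, upperarm_r links straight to spine_05.
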PhysicsAsset /BRCosmetics/Characters/Player/Male/Medium/Bodies/M_MED_SteelyGrin/Meshes/M_MED_SteelyGrin_jacket_collar_Physics
bodies {
 base_hood_l_01 [{
   shapeType=Capsule
   length=0.505
   radius=0.505
   shapeName=base_hood_l_01_capsule
   center=<-0.2287 -0.5678 -0.4415>
   rotation=<rot 18.6894 -90 -2.3955>
  }]
 base_hood_l_02 [{shapeType=Capsule, length=0.505, radius=0.505, shapeName=base_hood_l_02_capsule, rotation=<rot 0 -90 0>}]
 base_hood_r_01 [{shapeType=Capsule, length=0.505, radius=0.505, shapeName=base_hood_r_01_capsule, rotation=<rot 0 -90 0>}]
 base_hood_r_02 [{shapeType=Capsule, length=0.505, radius=0.505, shapeName=base_hood_r_02_capsule, rotation=<rot 0 -90 0>}]
 base_jacket_flap_l [{shapeType=Capsule, length=0.505, radius=0.505, shapeName=base_jacket_flap_l_capsule, rotation=<rot 0 -90 0>}]
 base_jacket_flap_l_02 [{shapeType=Capsule, length=0.505, radius=0.505, shapeName=base_jacket_flap_l_02_capsule, rotation=<rot 0 -90 0>}]
 base_jacket_flap_r [{shapeType=Capsule, length=0.505, radius=0.505, shapeName=base_jacket_flap_r_capsule, rotation=<rot 0 -90 0>}]
 base_jacket_flap_r_02 [{shapeType=Capsule, length=0.505, radius=0.505, shapeName=base_jacket_flap_r_02_capsule, rotation=<rot 0 -90 0>}]
 clavicle_l [{
   shapeType=Capsule
   length=3.774
   radius=2.8912
   shapeName=clavicle_l_capsule
   center=<14.6719 0.1001 2.4917>
   rotation=<rot 0 -172.7354 -78.2402>
  }]
 dyn_hood_L_01 [{shapeType=Capsule, length=2.1521, radius=2.1057, shapeName=dyn_hood_L_01_capsule, center=<3.2918 0 -0.0928>, rotation=<rot 90 0 0>}]
 dyn_hood_L_02 [
  {
   shapeType=Capsule
   length=7.6621
   radius=1.1436
   shapeName=dyn_hood_L_02_capsule
   center=<5.8704 0 0.016>
   rotation=<rot 86.6735 -179.9998 -179.9998>
  }
  {
   shapeType=Capsule
   length=4.3964
   radius=1.0446
   shapeName=dyn_hood_L_02_capsule
   center=<4.5052 -3.0504 -0.2303>
   rotation=<rot 88.417 -0.0017 4.3432>
  }
  {
   shapeType=Capsule
   length=10.0643
   radius=0.9225
   shapeName=dyn_hood_L_02_capsule
   center=<7.0828 3.0095 -0.1048>
   rotation=<rot 84.3066 -179.9996 -179.9995>
  }
 ]
 dyn_hood_r_01 [{shapeType=Capsule, length=1.7401, radius=2.1747, shapeName=dyn_hood_r_01_capsule, center=<-3.366 0 0>, rotation=<rot 90 0 0>}]
 dyn_hood_r_02 [
  {
   shapeType=Capsule
   length=8.6595
   radius=0.8729
   shapeName=dyn_hood_r_02_capsule
   center=<-6.5821 0 -0.0458>
   rotation=<rot 83.234 -179.9998 -177.306>
  }
  {
   shapeType=Capsule
   length=5.7664
   radius=1.1227
   shapeName=dyn_hood_r_02_capsule
   center=<-5.5981 3.4034 -0.0769>
   rotation=<rot 84.0514 -179.876 -175.1092>
  }
  {
   shapeType=Capsule
   length=11.1747
   radius=1.012
   shapeName=dyn_hood_r_02_capsule
   center=<-7.403 -2.7582 0.0285>
   rotation=<rot 83.1493 -179.9996 -178.1249>
  }
 ]
 dyn_jacket_flap_l [{
   shapeType=Capsule
   length=6.8365
   radius=0.9696
   shapeName=dyn_jacket_flap_l_capsule
   center=<2.66 0.6089 0.5381>
   rotation=<rot 6.4612 -88.8227 25.074>
  }]
 dyn_jacket_flap_l_02 [{
   shapeType=Capsule
   length=6.1537
   radius=0.9395
   shapeName=dyn_jacket_flap_l_02_capsule
   center=<0.6752 0.156 0.1077>
   rotation=<rot 5.5447 -89.8152 10.1242>
  }]
 dyn_jacket_flap_r [
  {
   shapeType=Capsule
   length=7.2899
   radius=1.3064
   shapeName=dyn_jacket_flap_r_capsule
   center=<3.6332 -1.2093 -0.0646>
   rotation=<rot 7.2434 94.4711 -23.2471>
  }
  {
   shapeType=Capsule
   length=3.5701
   radius=0.9554
   shapeName=dyn_jacket_flap_r_capsule
   center=<1.6757 -0.4031 2.6656>
   rotation=<rot -21.1942 0 0.0001>
  }
 ]
 dyn_jacket_flap_r_02 [{
   shapeType=Capsule
   length=5.2995
   radius=1.1205
   shapeName=dyn_jacket_flap_r_02_capsule
   center=<0.6059 0.4051 -1.5232>
   rotation=<rot -0.0688 -90 9.3849>
  }]
 head [{shapeType=Capsule, length=3.7374, radius=8.329, shapeName=head_capsule, center=<2.0138 2.6053 0.0015>, rotation=<rot 0 -90 89.9585>}]
 neck_02 [{shapeType=Capsule, length=1.5201, radius=4.8784, shapeName=neck_02_capsule, center=<3.7796 5.4516 0>, rotation=<rot 0 -90 0>}]
 spine_05 [{shapeType=Capsule, length=20.5669, radius=9.3599, center=<-3.1196 1.9728 -0.0328>, rotation=<rot 0.3193 -90 -0.124>}]
 upperarm_l [{
   shapeType=Capsule
   length=12.908
   radius=6.5121
   shapeName=upperarm_l_capsule
   center=<12.1482 0 -0.5935>
   rotation=<rot 84.8434 -179.9993 -179.9993>
  }]
 upperarm_r [{
   shapeType=Capsule
   length=12.0597
   radius=5.9448
   shapeName=upperarm_r_capsule
   center=<-12.1482 0 0.5935>
   rotation=<rot 84.8434 -179.9993 0.0007>
  }]
}
constraints [
 {bone1Name=dyn_jacket_flap_l, bone2Name=base_jacket_flap_l, swing1LimitDegrees=25, swing2LimitDegrees=0, twistLimitDegrees=0}
 {bone1Name=dyn_jacket_flap_r, bone2Name=base_jacket_flap_r, swing1LimitDegrees=45, swing2LimitDegrees=5, twistLimitDegrees=2}
 {bone1Name=dyn_jacket_flap_l_02, bone2Name=base_jacket_flap_l_02, swing1LimitDegrees=20, swing2LimitDegrees=5, twistLimitDegrees=2}
 {bone1Name=dyn_jacket_flap_r_02, bone2Name=base_jacket_flap_r_02, swing1LimitDegrees=20, swing2LimitDegrees=5, twistLimitDegrees=2}
 {
  bone1Name=base_jacket_flap_l
  bone2Name=spine_05
  swing1LimitDegrees=45
  swing1Motion=Free
  swing2LimitDegrees=45
  swing2Motion=Free
  twistLimitDegrees=45
  twistMotion=Free
 }
 {
  bone1Name=base_jacket_flap_r
  bone2Name=spine_05
  swing1LimitDegrees=45
  swing1Motion=Free
  swing2LimitDegrees=45
  swing2Motion=Free
  twistLimitDegrees=45
  twistMotion=Free
 }
 {
  bone1Name=upperarm_l
  bone2Name=clavicle_l
  swing1LimitDegrees=45
  swing1Motion=Free
  swing2LimitDegrees=45
  swing2Motion=Free
  twistLimitDegrees=45
  twistMotion=Free
 }
 {
  bone1Name=upperarm_r
  bone2Name=spine_05
  swing1LimitDegrees=45
  swing1Motion=Free
  swing2LimitDegrees=45
  swing2Motion=Free
  twistLimitDegrees=45
  twistMotion=Free
 }
 {bone1Name=dyn_hood_r_02, bone2Name=base_hood_r_02, swing1LimitDegrees=10, swing2LimitDegrees=10, twistLimitDegrees=0}
 {
  bone1Name=neck_02
  bone2Name=spine_05
  swing1LimitDegrees=45
  swing1Motion=Free
  swing2LimitDegrees=45
  swing2Motion=Free
  twistLimitDegrees=45
  twistMotion=Free
 }
 {bone1Name=dyn_hood_L_02, bone2Name=base_hood_l_02, swing1LimitDegrees=5, swing2LimitDegrees=10, twistLimitDegrees=0}
 {
  bone1Name=base_hood_l_02
  bone2Name=spine_05
  swing1LimitDegrees=45
  swing1Motion=Free
  swing2LimitDegrees=45
  swing2Motion=Free
  twistLimitDegrees=45
  twistMotion=Free
 }
 {
  bone1Name=base_hood_r_02
  bone2Name=spine_05
  swing1LimitDegrees=45
  swing1Motion=Free
  swing2LimitDegrees=45
  swing2Motion=Free
  twistLimitDegrees=45
  twistMotion=Free
 }
 {
  bone1Name=clavicle_l
  bone2Name=spine_05
  swing1LimitDegrees=45
  swing1Motion=Free
  swing2LimitDegrees=45
  swing2Motion=Free
  twistLimitDegrees=45
  twistMotion=Free
 }
 {
  bone1Name=base_hood_l_01
  bone2Name=spine_05
  swing1LimitDegrees=45
  swing1Motion=Free
  swing2LimitDegrees=45
  swing2Motion=Free
  twistLimitDegrees=45
  twistMotion=Free
 }
 {
  bone1Name=base_hood_r_01
  bone2Name=spine_05
  swing1LimitDegrees=45
  swing1Motion=Free
  swing2LimitDegrees=45
  swing2Motion=Free
  twistLimitDegrees=45
  twistMotion=Free
 }
 {bone1Name=dyn_hood_L_01, bone2Name=base_hood_l_01, swing1LimitDegrees=5, swing2LimitDegrees=2, twistLimitDegrees=0}
 {bone1Name=dyn_hood_r_01, bone2Name=base_hood_r_01, swing1LimitDegrees=5, swing2LimitDegrees=2, twistLimitDegrees=0}
 {
  bone1Name=head
  bone2Name=neck_02
  swing1LimitDegrees=45
  swing1Motion=Free
  swing2LimitDegrees=45
  swing2Motion=Free
  twistLimitDegrees=45
  twistMotion=Free
 }
]
