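# Physics asset for the necklace on the M_MED_SteelyGrin medium male player body, giving it jointed secondary motion.
# The chain runs from spine_05 through base_necklace_01 to dyn_necklace_01 and dyn_necklace_02, with free rotation at the spine joint and tight swing limits with zero twist below it.
# Three capsules shape the first dynamic link, a thin box shapes the second, and a broad, flat box on spine_05 provides an upper-torso collision surface.
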PhysicsAsset /BRCosmetics/Characters/Player/Male/Medium/Bodies/M_MED_SteelyGrin/Meshes/M_MED_SteelyGrin_necklace_Physics
bodies {
 base_necklace_01 [{shapeType=Capsule, length=0.505, radius=0.505, shapeName=base_necklace_01_capsule, rotation=<rot 0 -90 0>}]
 dyn_necklace_01 [
  {shapeType=Capsule, length=3.7976, radius=1.0943, shapeName=dyn_necklace_01_capsule, center=<2.465 0.4215 0>, rotation=<rot 90 0 0>}
  {
   shapeType=Capsule
   length=4.0129
   radius=0.5462
   shapeName=dyn_necklace_01_capsule
   center=<2.2975 -0.0709 -2.2887>
   rotation=<rot -62.6973 0.0001 0>
  }
  {
   shapeType=Capsule
   length=6.0042
   radius=0.5131
   shapeName=dyn_necklace_01_capsule
   center=<2.061 -0.1845 2.106>
   rotation=<rot -62.9932 -180 180>
  }
 ]
 dyn_necklace_02 [{
   shapeType=Box
   extentX=0.534
   extentY=2.4351
   extentZ=3.4219
   shapeName=dyn_necklace_02_box
   center=<1.2686 0.2815 0>
   rotation=<rot 0 98.3284 -89.9998>
  }]
 spine_05 [{
   shapeType=Box
   extentX=17.2946
   extentY=12.7748
   extentZ=1.818
   shapeName=spine_05_box
   center=<3.1891 7.5465 -0.1289>
   rotation=<rot 2.3781 147.2778 -89.5513>
  }]
}
constraints [
 {bone1Name=dyn_necklace_01, bone2Name=base_necklace_01, swing1LimitDegrees=45, swing2LimitDegrees=2, twistLimitDegrees=0}
 {
  bone1Name=base_necklace_01
  bone2Name=spine_05
  swing1LimitDegrees=45
  swing1Motion=Free
  swing2LimitDegrees=45
  swing2Motion=Free
  twistLimitDegrees=45
  twistMotion=Free
 }
 {bone1Name=dyn_necklace_02, bone2Name=dyn_necklace_01, swing1LimitDegrees=15, swing2LimitDegrees=10, twistLimitDegrees=0}
]
