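# Post-process-style AnimBlueprint for the SteelyGrin player cosmetic body, adding secondary motion to its hood, grenade, necklace, back chains and jacket.
# Copies the parent mesh pose into a Control Rig, then chains cached passes: constraints and additive bone offsets, AnimDynamics and RigidBody simulation, final bone corrections.
# The event graph tracks weapon, targeting, jump, melee, glider and emote states, and derives sim-space and collar gravity settings from a deprecated state function.
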AnimBlueprint /BRCosmetics/Characters/Player/Male/Medium/Bodies/M_MED_SteelyGrin/Meshes/M_MED_SteelyGrin_AnimBP
target_skeleton=&/M_MED_SteelyGrin_Skeleton
graphs [
 AnimGraph {nodes [
   Output Pose(Result=Use cached pose 'CRCache_postDyn'()) {layerGroup=None}
   n1 = Control Rig!(
    Source=Copy Pose From Mesh() {bUseAttachedParent=true}
    Alpha
    GenericEmotesPlaying:bool=false
    Is Crouching:bool=Get bIsCrouching()
    Is targeting:double=Get isTargeting float() -> (isTargeting float:double)
    DualWieldSwing:bool=Get Dual wield Swing() -> (Dual wield Swing:bool)
    Is Playing Emotes:bool=Get bIsPlayingEmote()
    fIsJumping:double=Get fIsJumping() -> (fIsJumping:double)
    Large weapon out:bool=Get Large weapon out() -> (Large weapon out:bool)
    SMG out:bool=Get Smg_Out() -> (Smg_Out:bool)
    Mime Emote Play:bool=false
   ) -> (Pose) {
    alphaBoolBlend {blendOption=Linear}
    controlRigAssetReference {blueprintRigClass=&/M_MED_SteelyGrin_CtrlRig.M_MED_SteelyGrin_CtrlRig_C}
   }
   Save cached pose 'CRCache_inputPose'(Pose=@n1) {cacheName=CRCache_inputPose}
   n2 = Anim Dynamics - Bone: dyn_hood_back_01!(
    ComponentPose=Local To Component(LocalPose=Use cached pose 'CRCache_preDyn'())
    Alpha=1
    GravityOverride:Vector="0,0,-980"
    ExternalForce="0,0,0"
   ) -> (Pose) {
    angularDampingOverride=1
    angularSpringConstant=20000
    bAngularSpring=true
    bOverrideAngularDamping=true
    bUseGravityOverride=true
    gravityScale=0.35
    simulationSpace=RootRelative
    alphaBoolBlend {blendOption=Linear}
    boundBone {boneName=dyn_hood_back_01}
    componentLinearAccScale=<1 1 1>
    componentLinearVelScale=<1 1 1>
    gravityOverride {z=-980}
    physicsBodyDefinitions [{
      boundBone {boneName=dyn_hood_back_01}
      boxExtents=<20 30 30>
      constraintSetup {angularLimitsMax=<2 2 5>, angularLimitsMin=<-2 -2 -5>, angularTarget=<1 0 0>}
      localJointOffset=<2 0 0>
     }]
   }
   n3 = Anim Dynamics - Bone: dyn_nade!(ComponentPose=@n2, Alpha=1, GravityOverride:Vector="0,0,-980", ExternalForce="0,0,0") -> (Pose) {
    angularDampingOverride=1
    angularSpringConstant=20000
    bAngularSpring=true
    bOverrideAngularDamping=true
    bUseGravityOverride=true
    bUseSphericalLimits=true
    gravityScale=0.35
    simulationSpace=RootRelative
    alphaBoolBlend {blendOption=Linear}
    boundBone {boneName=dyn_nade}
    componentLinearAccScale=<1 1 1>
    componentLinearVelScale=<1 1 1>
    gravityOverride {z=-980}
    physicsBodyDefinitions [{
      boundBone {boneName=dyn_nade}
      boxExtents=<20 30 30>
      constraintSetup {angularLimitsMax=<25 30 15>, angularLimitsMin=<-5 0 -15>, angularTarget=<1 0 0>}
      localJointOffset=<3 0 0>
     }]
    sphericalLimits [{limitRadius=12.5, drivingBone {boneName=thigh_r}, sphereLocalOffset=<12 0 -1.715>}]
   }
   n4 = RigidBody!(ComponentPose=@n3, Alpha, OverrideWorldGravity:Vector="0,0,-500", ExternalForce) -> (Pose) {
    bOverrideWorldGravity=true
    overridePhysicsAsset=&/M_MED_SteelyGrin_necklace_Physics
    simulationSpace=BaseBoneSpace
    alphaBoolBlend {blendOption=Linear}
    baseBoneRef {boneName=spine_05}
    componentAppliedLinearAccClamp=<2998 3000 3000>
    overrideWorldGravity {z=-500}
    simSpaceSettings {
     dampingAlpha=0
     maxAngularAcceleration=40
     maxAngularVelocity=20
     maxLinearAcceleration=15000
     maxLinearVelocity=700
     velocityScaleZ=0
     worldAlpha=1
     externalLinearDragV=<3 3 1.5>
    }
   }
   n5 = Anim Dynamics - Bone: base_backChain_A_01!(ComponentPose=@n4, Alpha=1, GravityOverride:Vector="0,0,0", ExternalForce="0,0,0") -> (Pose) {
    angularDampingOverride=1
    angularSpringConstant=20000
    bAngularSpring=true
    bOverrideAngularDamping=true
    bUseGravityOverride=true
    simulationSpace=RootRelative
    alphaBoolBlend {blendOption=Linear}
    boundBone {boneName=base_backChain_A_01}
    componentLinearAccScale=<1 1 1.3>
    componentLinearVelScale=<1 1 1.3>
    physicsBodyDefinitions [{
      boundBone {boneName=base_backChain_A_01}
      boxExtents=<20 30 30>
      constraintSetup {angularLimitsMax=<0 5 1>, angularLimitsMin=<0 -5 -60>, angularTarget=<1 0 0>}
      localJointOffset=<12 0 0>
     }]
   }
   n6 = Anim Dynamics - Bone: base_backChain_A_02!(ComponentPose=@n5, Alpha=1, GravityOverride:Vector="0,0,0", ExternalForce="0,0,0") -> (Pose) {
    angularDampingOverride=1
    angularSpringConstant=20000
    bAngularSpring=true
    bOverrideAngularDamping=true
    bUseGravityOverride=true
    simulationSpace=RootRelative
    alphaBoolBlend {blendOption=Linear}
    boundBone {boneName=base_backChain_A_02}
    componentLinearAccScale=<1 1 1.3>
    componentLinearVelScale=<1 1 1.3>
    physicsBodyDefinitions [{
      boundBone {boneName=base_backChain_A_02}
      boxExtents=<20 30 30>
      constraintSetup {angularLimitsMax=<0 5 1>, angularLimitsMin=<0 -5 -60>, angularTarget=<1 0 0>}
      localJointOffset=<12 0 0>
     }]
   }
   n7 = Anim Dynamics - Bone: base_backChain_B_01!(ComponentPose=@n6, Alpha=1, GravityOverride:Vector="0,0,0", ExternalForce="0,0,0") -> (Pose) {
    angularDampingOverride=1
    angularSpringConstant=20000
    bAngularSpring=true
    bOverrideAngularDamping=true
    bUseGravityOverride=true
    simulationSpace=RootRelative
    alphaBoolBlend {blendOption=Linear}
    boundBone {boneName=base_backChain_B_01}
    componentLinearAccScale=<1 1 1.3>
    componentLinearVelScale=<1 1 1.3>
    physicsBodyDefinitions [{
      boundBone {boneName=base_backChain_B_01}
      boxExtents=<20 30 30>
      constraintSetup {angularLimitsMax=<0 5 1>, angularLimitsMin=<0 -5 -60>, angularTarget=<1 0 0>}
      localJointOffset=<12 0 0>
     }]
   }
   n8 = Anim Dynamics - Bone: base_backChain_B_02!(ComponentPose=@n7, Alpha=1, GravityOverride:Vector="0,0,0", ExternalForce="0,0,0") -> (Pose) {
    angularDampingOverride=1
    angularSpringConstant=20000
    bAngularSpring=true
    bOverrideAngularDamping=true
    bUseGravityOverride=true
    simulationSpace=RootRelative
    alphaBoolBlend {blendOption=Linear}
    boundBone {boneName=base_backChain_B_02}
    componentLinearAccScale=<1 1 1.3>
    componentLinearVelScale=<1 1 1.3>
    physicsBodyDefinitions [{
      boundBone {boneName=base_backChain_B_02}
      boxExtents=<20 30 30>
      constraintSetup {angularLimitsMax=<0 5 1>, angularLimitsMin=<0 -5 -60>, angularTarget=<1 0 0>}
      localJointOffset=<12 0 0>
     }]
   }
   n9 = RigidBody(ComponentPose=@n8, OverrideWorldGravity="0,0,0") {
    bOverrideWorldGravity=true
    overridePhysicsAsset=&/M_MED_SteelyGrin_Physics
    simulationSpace=BaseBoneSpace
    alphaBoolBlend {blendOption=Linear}
    baseBoneRef {boneName=spine_01}
    componentAppliedLinearAccClamp=<3000 3000 3000>
    simSpaceSettings {
     dampingAlpha=0
     maxAngularAcceleration=20
     maxAngularVelocity=10
     maxLinearAcceleration=1000
     maxLinearVelocity=700
     velocityScaleZ=0.5
     worldAlpha=1
     externalLinearDragV=<1 1 1>
    }
   }
   n10 = RigidBody(ComponentPose=@n9, OverrideWorldGravity="0,0,0") {
    bOverrideWorldGravity=true
    overridePhysicsAsset=&/M_MED_SteelyGrin_jacket_collar_Physics
    simulationSpace=BaseBoneSpace
    alphaBoolBlend {blendOption=Linear}
    baseBoneRef {boneName=spine_05}
    componentAppliedLinearAccClamp=<3000 3000 3000>
    simSpaceSettings {
     dampingAlpha=0
     maxAngularAcceleration=20
     maxAngularVelocity=10
     maxLinearAcceleration=1000
     maxLinearVelocity=700
     velocityScaleZ=0.5
     worldAlpha=1
     externalLinearDragV=<1 1 1>
    }
   }
   Save cached pose 'CRCache_dynamics'(Pose=Component To Local(ComponentPose=@n10)) {cacheName=CRCache_dynamics}
   n11 = "Transform (Modify) Bone - Bone: dyn_nade"!(
    ComponentPose=Local To Component(LocalPose=Use cached pose 'CRCache_dynamics'())
    bAlphaBoolEnabled:bool=True
    Translation="-0.818,-0.000,-0.412"
    Rotation="20.000,53.204,-30.000"
   ) -> (Pose) {
    alphaInputType=Bool
    rotationMode=BMM_Additive
    translationMode=BMM_Additive
    alphaBoolBlend {blendInTime=0.1, blendOption=QuadraticInOut, blendOutTime=1}
    boneToModify {boneName=dyn_nade}
    rotation=<rot 20 53.204 -30>
    translation {x=-0.818, z=-0.412}
   }
   n12 = "Transform (Modify) Bone - Bone: dyn_nade"(ComponentPose=@n11, bAlphaBoolEnabled=True, Translation="-2.953,0.000,2.148", Rotation="-7.124,51.426,-61.363") {
    alphaInputType=Bool
    rotationMode=BMM_Additive
    translationMode=BMM_Additive
    alphaBoolBlend {blendInTime=0.1, blendOption=QuadraticInOut, blendOutTime=1}
    boneToModify {boneName=dyn_nade}
    rotation=<rot -7.124 51.426 -61.363>
    translation {x=-2.953, z=2.148}
   }
   n13 = "Transform (Modify) Bone - Bone: dyn_low_jacket_back"!(ComponentPose=@n12, Alpha=1, Translation="0,-2,0") -> (Pose) {
    rotationMode=BMM_Additive
    translationMode=BMM_Additive
    alphaBoolBlend {blendOption=Linear}
    boneToModify {boneName=dyn_low_jacket_back}
    translation {y=-2}
   }
   n14 = "Transform (Modify) Bone - Bone: dyn_low_jacket_r_02"!(ComponentPose=@n13, Alpha=1, Translation="-1.548,-1,-0.137") -> (Pose) {
    rotationMode=BMM_Additive
    translationMode=BMM_Additive
    alphaBoolBlend {blendOption=Linear}
    boneToModify {boneName=dyn_low_jacket_r_02}
    translation=<-1.548 -1 -0.137>
   }
   n15 = "Transform (Modify) Bone - Bone: dyn_low_jacket_l_02"!(ComponentPose=@n14, Alpha=1, Translation="1,-2,0") -> (Pose) {
    rotationMode=BMM_Additive
    translationMode=BMM_Additive
    alphaBoolBlend {blendInTime=0.1, blendOption=Linear, blendOutTime=0.1}
    boneToModify {boneName=dyn_low_jacket_l_02}
    translation=<1 -2>
   }
   n16 = "Transform (Modify) Bone - Bone: lowerarm_twist_02_l"!(ComponentPose=@n15, bAlphaBoolEnabled:bool=True, Translation="0,0,0", Rotation="0,0,0", Scale="0.1,0.1,0.1") -> (Pose) {
    alphaInputType=Bool
    rotationMode=BMM_Additive
    rotationSpace=BCS_ParentBoneSpace
    scaleMode=BMM_Additive
    scaleSpace=BCS_ParentBoneSpace
    translationMode=BMM_Additive
    translationSpace=BCS_WorldSpace
    alphaBoolBlend {blendOption=Linear}
    boneToModify {boneName=lowerarm_twist_02_l}
    scale=<0.1 0.1 0.1>
   }
   n17 = "Transform (Modify) Bone - Bone: lowerarm_twist_02_r"!(ComponentPose=@n16, bAlphaBoolEnabled:bool=True, Translation="0,0,0", Rotation="0,0,0", Scale="1,1,1") -> (Pose) {
    alphaInputType=Bool
    rotationMode=BMM_Additive
    rotationSpace=BCS_BoneSpace
    scaleMode=BMM_Additive
    scaleSpace=BCS_BoneSpace
    translationMode=BMM_Additive
    translationSpace=BCS_BoneSpace
    alphaBoolBlend {blendOption=Linear}
    boneToModify {boneName=lowerarm_twist_02_r}
   }
   Save cached pose 'CRCache_postDyn'(Pose=Component To Local(ComponentPose=@n17)) {cacheName=CRCache_postDyn}
   n18 = "Constraint - base_backChain_B_01 (PPP)"!(
    ComponentPose=Local To Component(LocalPose=Use cached pose 'CRCache_inputPose'())
    Alpha=1
    ConstraintWeights_0=0.7135
    ConstraintWeights_1:float=0.2776
    ConstraintWeights_2:float=0.009
   ) -> (Pose) {
    alphaBoolBlend {blendOption=Linear}
    boneToModify {boneName=base_backChain_B_01}
    constraintSetup [{targetBone {boneName=pelvis}}, {targetBone {boneName=thigh_twist_01_r}}, {targetBone {boneName=spine_01}}]
    constraintWeights [1,1,1]
   }
   n19 = "Constraint - base_backChain_A_01 (PPP)"!(ComponentPose=@n18, Alpha=1, ConstraintWeights_0=0.7439, ConstraintWeights_1:float=0.2495, ConstraintWeights_2:float=0.0066) -> (Pose) {
    alphaBoolBlend {blendOption=Linear}
    boneToModify {boneName=base_backChain_A_01}
    constraintSetup [{targetBone {boneName=pelvis}}, {targetBone {boneName=thigh_twist_01_r}}, {targetBone {boneName=spine_01}}]
    constraintWeights [1,1,1]
   }
   n20 = "Constraint - base_backChain_B_02 (PPP)"!(ComponentPose=@n19, Alpha=1, ConstraintWeights_0=0.7841, ConstraintWeights_1:float=0.1512, ConstraintWeights_2:float=0.0648) -> (Pose) {
    alphaBoolBlend {blendOption=Linear}
    boneToModify {boneName=base_backChain_B_02}
    constraintSetup [{targetBone {boneName=pelvis}}, {targetBone {boneName=thigh_twist_01_r}}, {targetBone {boneName=spine_01}}]
    constraintWeights [1,1,1]
   }
   n21 = "Constraint - base_backChain_A_02 (PPP)"!(ComponentPose=@n20, Alpha=1, ConstraintWeights_0=0.7073, ConstraintWeights_1:float=0.2314, ConstraintWeights_2:float=0.0613) -> (Pose) {
    alphaBoolBlend {blendOption=Linear}
    boneToModify {boneName=base_backChain_A_02}
    constraintSetup [{targetBone {boneName=pelvis}}, {targetBone {boneName=thigh_twist_01_r}}, {targetBone {boneName=spine_01}}]
    constraintWeights [1,1,1]
   }
   n22 = "Transform (Modify) Bone - Bone: base_backChain_B_01"!(ComponentPose=@n21, bAlphaBoolEnabled:bool=True, Translation="0,0,0", Rotation="0,0,5") -> (Pose) {
    alphaInputType=Bool
    rotationMode=BMM_Additive
    translationMode=BMM_Additive
    alphaBoolBlend {blendOption=Linear}
    boneToModify {boneName=base_backChain_B_01}
    rotation {roll=5}
   }
   n23 = "Transform (Modify) Bone - Bone: dyn_backChain_B_02"!(ComponentPose=@n22, bAlphaBoolEnabled:bool=True, Translation="0,0,0", Rotation="0,0,5") -> (Pose) {
    alphaInputType=Bool
    rotationMode=BMM_Additive
    translationMode=BMM_Additive
    alphaBoolBlend {blendOption=Linear}
    boneToModify {boneName=dyn_backChain_B_02}
    rotation {roll=5}
   }
   n24 = "Transform (Modify) Bone - Bone: dyn_backChain_A_02"!(ComponentPose=@n23, bAlphaBoolEnabled:bool=True, Translation="0,0,0", Rotation="0,0,5") -> (Pose) {
    alphaInputType=Bool
    rotationMode=BMM_Additive
    translationMode=BMM_Additive
    alphaBoolBlend {blendOption=Linear}
    boneToModify {boneName=dyn_backChain_A_02}
    rotation {roll=5}
   }
   n25 = "Transform (Modify) Bone - Bone: dyn_backChain_A_01"!(ComponentPose=@n24, bAlphaBoolEnabled:bool=True, Translation="0,0,0", Rotation="0,0,5") -> (Pose) {
    alphaInputType=Bool
    rotationMode=BMM_Additive
    translationMode=BMM_Additive
    alphaBoolBlend {blendOption=Linear}
    boneToModify {boneName=dyn_backChain_A_01}
    rotation {roll=5}
   }
   n26 = "Transform (Modify) Bone - Bone: base_low_jacket_l_01"!(ComponentPose=@n25, Alpha=1, Translation="0,2,3", Rotation="0,0,-15") -> (Pose) {
    rotationMode=BMM_Additive
    scaleMode=BMM_Additive
    translationMode=BMM_Additive
    alphaBoolBlend {blendOption=Linear}
    alphaScaleBiasClamp {bInterpResult=true, interpSpeedIncreasing=20}
    boneToModify {boneName=base_low_jacket_l_01}
    rotation {roll=-15}
    translation {y=2, z=3}
   }
   n27 = "Transform (Modify) Bone - Bone: base_low_jacket_r_01"!(ComponentPose=@n26, Alpha=1, Translation="-1.5,2,2", Rotation="-10,0,10") -> (Pose) {
    rotationMode=BMM_Additive
    scaleMode=BMM_Additive
    translationMode=BMM_Additive
    alphaBoolBlend {blendInTime=0.1, blendOption=Linear, blendOutTime=0.3}
    alphaScaleBiasClamp {bInterpResult=true, interpSpeedIncreasing=35}
    boneToModify {boneName=base_low_jacket_r_01}
    rotation {pitch=-10, roll=10}
    translation=<-1.5 2 2>
   }
   n28 = "Transform (Modify) Bone - Bone: base_low_jacket_r_01"(ComponentPose=@n27, Alpha=1, Translation="-0.5,0,0", Rotation="-10,0,0") {
    rotationMode=BMM_Additive
    translationMode=BMM_Additive
    alphaBoolBlend {blendOption=Linear}
    boneToModify {boneName=base_low_jacket_r_01}
    rotation {pitch=-10}
    translation {x=-0.5}
   }
   n29 = "Transform (Modify) Bone - Bone: base_jacket_flap_l"!(ComponentPose=@n28, bAlphaBoolEnabled:bool=True, Translation="-2,0,0", Rotation="10,0,0") -> (Pose) {
    alphaInputType=Bool
    rotationMode=BMM_Additive
    translationMode=BMM_Additive
    alphaBoolBlend {blendOption=Linear}
    boneToModify {boneName=base_jacket_flap_l}
    rotation {pitch=10}
    translation {x=-2}
   }
   n30 = "Transform (Modify) Bone - Bone: base_low_jacket_back"!(ComponentPose=@n29, bAlphaBoolEnabled:bool=True, Translation="0,-2,0", Rotation="0,-10,0") -> (Pose) {
    alphaInputType=Bool
    rotationMode=BMM_Additive
    rotationSpace=BCS_BoneSpace
    scaleSpace=BCS_BoneSpace
    translationMode=BMM_Additive
    translationSpace=BCS_BoneSpace
    alphaBoolBlend {blendInTime=0.5, blendOption=Linear, blendOutTime=1}
    boneToModify {boneName=base_low_jacket_back}
    rotation {yaw=-10}
    translation {y=-2}
   }
   n31 = "Transform (Modify) Bone - Bone: base_low_jacket_l_02"!(ComponentPose=@n30, bAlphaBoolEnabled:bool=True, Translation="0,-2,0", Rotation="0,-5,5") -> (Pose) {
    alphaInputType=Bool
    rotationMode=BMM_Additive
    rotationSpace=BCS_BoneSpace
    scaleSpace=BCS_BoneSpace
    translationMode=BMM_Additive
    translationSpace=BCS_BoneSpace
    alphaBoolBlend {blendInTime=0.1, blendOption=Linear, blendOutTime=0.1}
    boneToModify {boneName=base_low_jacket_l_02}
    rotation {roll=5, yaw=-5}
    translation {y=-2}
   }
   n32 = "Transform (Modify) Bone - Bone: base_low_jacket_r_02"!(ComponentPose=@n31, bAlphaBoolEnabled:bool=True, Translation="0,1,0", Rotation="0,-5,5") -> (Pose) {
    alphaInputType=Bool
    rotationMode=BMM_Additive
    rotationSpace=BCS_BoneSpace
    scaleSpace=BCS_BoneSpace
    translationMode=BMM_Additive
    translationSpace=BCS_BoneSpace
    alphaBoolBlend {blendInTime=0.1, blendOption=Linear, blendOutTime=0.1}
    boneToModify {boneName=base_low_jacket_r_02}
    rotation {roll=5, yaw=-5}
    translation {y=1}
   }
   n33 = "Transform (Modify) Bone - Bone: base_low_jacket_back"(ComponentPose=@n32, bAlphaBoolEnabled=True, Translation="0,-2,0", Rotation="0,-10,0") {
    alphaInputType=Bool
    rotationMode=BMM_Additive
    rotationSpace=BCS_BoneSpace
    scaleSpace=BCS_BoneSpace
    translationMode=BMM_Additive
    translationSpace=BCS_BoneSpace
    alphaBoolBlend {blendInTime=0.1, blendOption=Linear, blendOutTime=0.1}
    boneToModify {boneName=base_low_jacket_back}
    rotation {yaw=-10}
    translation {y=-2}
   }
   n34 = "Transform (Modify) Bone - Bone: base_low_jacket_l_02"(ComponentPose=@n33, bAlphaBoolEnabled=True, Translation="0,-2,0", Rotation="0,-5,5") {
    alphaInputType=Bool
    rotationMode=BMM_Additive
    rotationSpace=BCS_BoneSpace
    scaleSpace=BCS_BoneSpace
    translationMode=BMM_Additive
    translationSpace=BCS_BoneSpace
    alphaBoolBlend {blendInTime=0.1, blendOption=Linear, blendOutTime=0.1}
    boneToModify {boneName=base_low_jacket_l_02}
    rotation {roll=5, yaw=-5}
    translation {y=-2}
   }
   n35 = "Transform (Modify) Bone - Bone: base_low_jacket_r_02"(ComponentPose=@n34, bAlphaBoolEnabled=True, Translation="0,1,0", Rotation="0,-5,5") {
    alphaInputType=Bool
    rotationMode=BMM_Additive
    rotationSpace=BCS_BoneSpace
    scaleSpace=BCS_BoneSpace
    translationMode=BMM_Additive
    translationSpace=BCS_BoneSpace
    alphaBoolBlend {blendInTime=0.1, blendOption=Linear, blendOutTime=0.1}
    boneToModify {boneName=base_low_jacket_r_02}
    rotation {roll=5, yaw=-5}
    translation {y=1}
   }
   Save cached pose 'CRCache_preDyn'(Pose=Component To Local(ComponentPose=@n35)) {cacheName=CRCache_preDyn}
  ]}
 EventGraph {nodes [
   n1 = Event BlueprintUpdateAnimation()
   > n2 = Sequence!(execute) -> (then_0, then_1, then_2:Exec, then_3:Exec)
   > n3 = Gravity Override State Function DEPRECATED(CharacterPartAnimInstance=Self-Reference(), GameTimeInSeconds=GetGameTimeInSeconds(), GravityOverrideParams=Get Gravity Override Params())
   > Set Sim Space Settings(Sim Space Settings:Sim Space Settings Structure=Break RBANSetting Struct!(RBANSettingStruct:"RBANSetting Struct Structure (by ref)"=@n3.RBAN SettingsOut) -> (
     ComponentLinearAccScale_28_#1:Vector=~
     ComponentLinearVelScale_31_#2:Vector=~
     SimSpaceSettings_34_#3:Sim Space Settings Structure
    ) {is=title K2Node_BreakStruct}.SimSpaceSettings_34_#3) -> (Output_Get:Sim Space Settings Structure)
   > n4 = Branch(Condition=Get Is using hang glider in parachute() -> (Is using hang glider in parachute:bool))
   > Set Gravity Override Collar(Gravity Override Collar:Vector="0,0,-980") -> (Output_Get:Vector)
   Set Gravity Override Collar(execute=@n4.else, Gravity Override Collar=@n3)
   n5 = Get bIsTargeting()
   Set isTargeting float(
    execute=@n2.then_1
    isTargeting float:double=FInterpTo(Current=Get isTargeting float(), Target="To Float (Boolean)"(InBool=@n5), DeltaTime=@n1.DeltaTimeX, InterpSpeed=15)
   ) -> (Output_Get:double)
   n6 = Get bIsPlayingMeleeAnim()
   > Set fMeleeSwing(fMeleeSwing:double=FInterpTo(Current=Get fMeleeSwing() -> (fMeleeSwing:double), Target="To Float (Boolean)"(InBool=@n6), DeltaTime=@n1.DeltaTimeX, InterpSpeed=1)) -> (Output_Get:double)
   n7 = Get bIsJumping()
   > Set fIsJumping(fIsJumping:double=FInterpTo(Current=Get fIsJumping(), Target="To Float (Boolean)"(InBool=@n7), DeltaTime=@n1.DeltaTimeX, InterpSpeed=1)) -> (Output_Get:double)
   n8 = GetAnimInstance(self=GetOwnerSkeletalMeshComponent())
   Is Valid(exec=@n2.then_2, InputObject=@n8)
   > Set Mime emote play(Mime emote play:bool=Montage_IsPlaying(self=@n8, Montage=/BRCosmetics/Animation/Game/MainPlayer/Emotes/Mime/Emote_Mime_CMM_M)) -> (Output_Get:bool)
   Event OnWeaponChanged()
   > Set Smg_Out(Smg_Out:bool=false) -> (Output_Get:bool)
   > Set ShotGun_Out(ShotGun_Out:bool=false) -> (Output_Get:bool)
   > Set AR_Out(AR_Out:bool=false) -> (Output_Get:bool)
   > Set Large weapon out(Large weapon out:bool=false) -> (Output_Get:bool)
   > Set Is dual weild(Is dual weild:bool=false) -> (Output_Get:bool)
   > n9 = Switch on EFortWeaponCoreAnimation!(execute, Selection:EFortWeaponCoreAnimation Enum=Get WeaponCoreAnim!() -> (WeaponCoreAnim)) -> (
    Melee:Exec
    Pistol:Exec
    Shotgun:Exec
    PaperBlueprint:Exec
    Rifle:Exec
    MeleeOneHand:Exec
    MachinePistol:Exec
    RocketLauncher:Exec
    GrenadeLauncher:Exec
    GoingCommando:Exec
    AssaultRifle:Exec
    TacticalShotgun:Exec
    SniperRifle:Exec
    TrapPlacement:Exec
    ShoulderLauncher:Exec
    AbilityDecoTool:Exec
    Crossbow:Exec
    C4:Exec
    RemoteControl:Exec
    DualWield:Exec
    AR_BullPup:Exec
    AR_ForwardGrip:Exec
    MedPackPaddles:Exec
    SMG_P90:Exec
    AR_DrumGun:Exec
    Consumable_Small:Exec
    Consumable_Large:Exec
    Balloon:Exec
    MountedTurret:Exec
    CreativeTool:Exec
    ExplosiveBow:Exec
    AshtonIndigo:Exec
    AshtonChicago:Exec
    MeleeDualWield:Exec
    Unarmed:Exec
   )
   Set AR_Out(execute=@n9.AssaultRifle, AR_Out=true)
   > Set Large weapon out(Large weapon out=true)
   Set Is dual weild(execute=@n9.MeleeDualWield, Is dual weild=true)
   Set Smg_Out(execute=@n9.SMG_P90, Smg_Out=true)
   Set ShotGun_Out(execute=@n9.TacticalShotgun, ShotGun_Out=true)
   > Set Large weapon out(Large weapon out=true)
   Event OnEndPickaxeSwing() {is=title K2Node_Event}
   > Set Dual wield Swing(Dual wield Swing:bool=false) -> (Output_Get:bool)
   Event OnParachuteOpened()
   > Switch on EGliderType(Selection=Get PlayerGliderType() -> (PlayerGliderType=HangGlider))
   > Set Is using hang glider in parachute(Is using hang glider in parachute:bool=true) -> (Output_Get:bool)
   Event OnParachuteClosed()
   > Set Is using hang glider in parachute(Is using hang glider in parachute=true)
   Event OnBeginPickaxeSwing() {is=title K2Node_Event}
   > Set Dual wield Swing(Dual wield Swing=Get Is dual weild() -> (Is dual weild:bool))
  ]}
]
variables [
 AR_Out {type=boolean}
 Dual wield Swing {type=boolean}
 fIsJumping {type=number}
 fMeleeSwing {type=number}
 Gravity Override Collar {type=Vector}
 Gravity Override Params {type=GravityOverrideParamsStruct}
 GravityOverride {type=Vector}
 Is dual weild {type=boolean}
 Is using hang glider in parachute {type=boolean}
 isTargeting float {type=number}
 Large weapon out {type=boolean}
 Mime emote play {type=boolean}
 RBAN Params {
  type=RBANParamsStruct
  default {
   dBNO {simSpaceSettings {
     dampingAlpha=0
     maxAngularAcceleration=20
     maxAngularVelocity=10
     maxLinearAcceleration=1000
     maxLinearVelocity=700
     velocityScaleZ=0.5
     externalLinearDragV=<1 1 1>
    }}
   dBNO_Carried {simSpaceSettings {
     dampingAlpha=0
     maxAngularAcceleration=20
     maxAngularVelocity=10
     maxLinearAcceleration=1000
     maxLinearVelocity=700
     velocityScaleZ=0.5
     externalLinearDragV=<1 1 1>
    }}
   emoteOrMelee {simSpaceSettings {
     dampingAlpha=0
     maxAngularAcceleration=20
     maxAngularVelocity=10
     maxLinearAcceleration=1000
     maxLinearVelocity=700
     velocityScaleZ=0.5
     externalLinearDragV=<1 1 1>
    }}
   front_End {simSpaceSettings {
     dampingAlpha=0
     maxAngularAcceleration=20
     maxAngularVelocity=10
     maxLinearAcceleration=1000
     maxLinearVelocity=700
     velocityScaleZ=0.5
     externalLinearDragV=<1 1 1>
    }}
   onGround_Crouch_Moving {simSpaceSettings {
     dampingAlpha=0
     maxAngularAcceleration=20
     maxAngularVelocity=10
     maxLinearAcceleration=1000
     maxLinearVelocity=700
     velocityScaleZ=0.5
     externalLinearDragV=<1 1 1>
    }}
   onGround_Falling {simSpaceSettings {
     dampingAlpha=0
     maxAngularAcceleration=20
     maxAngularVelocity=10
     maxLinearAcceleration=1000
     maxLinearVelocity=700
     velocityScaleZ=0.5
     externalLinearDragV=<1 1 1>
    }}
   onGround_Moving {simSpaceSettings {
     dampingAlpha=0
     maxAngularAcceleration=20
     maxAngularVelocity=10
     maxLinearAcceleration=1000
     maxLinearVelocity=700
     velocityScaleZ=0.5
     externalLinearDragV=<1 1 1>
    }}
   onGround_Standing {simSpaceSettings {
     dampingAlpha=0
     maxAngularAcceleration=20
     maxAngularVelocity=10
     maxLinearAcceleration=1000
     maxLinearVelocity=700
     velocityScaleZ=0.5
     externalLinearDragV=<1 1 1>
    }}
   rail_Grinding {simSpaceSettings {
     dampingAlpha=0
     maxAngularAcceleration=20
     maxAngularVelocity=10
     maxLinearAcceleration=1000
     maxLinearVelocity=700
     velocityScaleZ=0.5
     worldAlpha=1
     externalLinearDragV=<1 1 1>
    }}
   riding_Moving {simSpaceSettings {
     dampingAlpha=0
     maxAngularAcceleration=20
     maxAngularVelocity=10
     maxLinearAcceleration=1000
     maxLinearVelocity=700
     velocityScaleZ=0.5
     externalLinearDragV=<1 1 1>
    }}
   riding_Moving_Targeting {simSpaceSettings {
     dampingAlpha=0
     maxAngularAcceleration=20
     maxAngularVelocity=10
     maxLinearAcceleration=1000
     maxLinearVelocity=700
     velocityScaleZ=0.5
     externalLinearDragV=<1 1 1>
    }}
   riding_Standing {simSpaceSettings {
     dampingAlpha=0
     maxAngularAcceleration=20
     maxAngularVelocity=10
     maxLinearAcceleration=1000
     maxLinearVelocity=700
     velocityScaleZ=0.5
     externalLinearDragV=<1 1 1>
    }}
   skydive_DivingDown {simSpaceSettings {
     dampingAlpha=0
     maxAngularAcceleration=20
     maxAngularVelocity=10
     maxLinearAcceleration=1000
     maxLinearVelocity=700
     velocityScaleZ=0.5
     externalLinearDragV=<1 1 1>
    }}
   skydive_FloatingDown {
    componentLinearAccScale=<0 0.5 0>
    componentLinearVelScale=<0 0.5 0>
    simSpaceSettings {
     dampingAlpha=0
     maxAngularAcceleration=20
     maxAngularVelocity=10
     maxLinearAcceleration=1000
     maxLinearVelocity=700
     velocityScaleZ=0.5
     externalLinearDragV=<1 1 1>
    }
   }
   skydive_GoingUp {simSpaceSettings {
     dampingAlpha=0
     maxAngularAcceleration=20
     maxAngularVelocity=10
     maxLinearAcceleration=1000
     maxLinearVelocity=700
     velocityScaleZ=0.5
     externalLinearDragV=<1 1 1>
    }}
   skydive_Parachuting {
    componentLinearAccScale=<0 0 0>
    componentLinearVelScale=<0 0 0>
    simSpaceSettings {
     dampingAlpha=0
     maxAngularAcceleration=20
     maxAngularVelocity=10
     maxLinearAcceleration=0
     maxLinearVelocity=0
     velocityScaleZ=0
     worldAlpha=0
     externalLinearDragV=<0 0 0>
    }
   }
   skydive_Surfing {simSpaceSettings {
     dampingAlpha=0
     maxAngularAcceleration=20
     maxAngularVelocity=10
     maxLinearAcceleration=1000
     maxLinearVelocity=700
     velocityScaleZ=0.5
     externalLinearDragV=<1 1 1>
    }}
   swimming {simSpaceSettings {
     dampingAlpha=0
     maxAngularAcceleration=20
     maxAngularVelocity=10
     maxLinearAcceleration=1000
     maxLinearVelocity=700
     velocityScaleZ=0.5
     externalLinearDragV=<1 1 1>
    }}
   tactical_Sprint {simSpaceSettings {
     dampingAlpha=0
     maxAngularAcceleration=20
     maxAngularVelocity=10
     maxLinearAcceleration=1000
     maxLinearVelocity=700
     velocityScaleZ=0.5
     externalLinearDragV=<1 1 1>
    }}
   ziplining {simSpaceSettings {
     dampingAlpha=0
     maxAngularAcceleration=20
     maxAngularVelocity=10
     maxLinearAcceleration=1000
     maxLinearVelocity=700
     velocityScaleZ=0.5
     externalLinearDragV=<1 1 1>
    }}
  }
 }
 ShotGun_Out {type=boolean}
 Sim Space Settings {
  type=SimSpaceSettings
  default {
   dampingAlpha=1
   maxAngularAcceleration=10000
   maxAngularVelocity=10000
   maxLinearAcceleration=10000
   maxLinearVelocity=10000
   velocityScaleZ=1
   worldAlpha=0
   externalLinearDragV=<0 0 0>
  }
 }
 SkelMesh {type=SkeletalMeshComponent}
 Smg_Out {type=boolean}
 spine_rot {type=boolean}
]
