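# Post-process style AnimBlueprint for the SteelyGrin male face accessory part, adding secondary motion to front hair.
# The parent mesh pose runs through a Control Rig, main hair Anim Dynamics, weighted multi-target constraints on base hair bones, then per-strand dynamics, staged through cached poses.
# A hat dynamics node exists but is not connected to the output.
# Events compute GravityOverride and flag PistolOut for pistol or dual-wield weapons.
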AnimBlueprint /BRCosmetics/Characters/Player/Male/Medium/Bodies/M_MED_SteelyGrin/Meshes/Parts/M_MED_SteelyGrin_FaceAcc_AnimBP
target_skeleton=&/M_MED_SteelyGrin_FaceAcc_Skeleton
graphs [
 AnimGraph {nodes [
   Anim Dynamics - Bone: dyn_hat!(ComponentPose, Alpha=0.75, GravityOverride:Vector="0,0,-980", ExternalForce="0,0,0") -> (Pose) {
    angularDampingOverride=1
    angularSpringConstant=10000
    bAngularSpring=true
    bOverrideAngularDamping=true
    bUseGravityOverride=true
    gravityScale=0.35
    simulationSpace=BoneRelative
    alphaBoolBlend {blendOption=Linear}
    boundBone {boneName=dyn_hat}
    componentLinearAccScale=<1 1 1>
    componentLinearVelScale=<1 1 1>
    gravityOverride {z=-980}
    physicsBodyDefinitions [{
      boundBone {boneName=dyn_hat}
      boxExtents=<18 28 28>
      constraintSetup {angularLimitsMax=<3 3 2>, angularLimitsMin=<-3 -3 -2>, angularTarget=<1 0 0>}
      localJointOffset=<4 0 0>
     }]
    relativeSpaceBone {boneName=head}
   }
   Output Pose(Result=Use cached pose 'CRCache_secondaryDynamics'()) {layerGroup=None}
   n1 = Control Rig!(Source=Copy Pose From Mesh() {bUseAttachedParent=true}, Alpha, isPistolOut:bool=false) -> (Pose) {
    alphaBoolBlend {blendOption=Linear}
    controlRigAssetReference {blueprintRigClass=&/M_MED_SteelyGrin_FaceAcc_CtrlRig.M_MED_SteelyGrin_FaceAcc_CtrlRig_C}
   }
   Save cached pose 'CRCache_inputPose'(Pose=@n1) {cacheName=CRCache_inputPose}
   n2 = Anim Dynamics - Bone: dyn_frontHair_main_01!(
    ComponentPose=Local To Component(LocalPose=Use cached pose 'CRCache_inputPose'())
    Alpha=1
    GravityOverride:Vector="0,0,0"
    ExternalForce="0,0,0"
   ) -> (Pose) {
    angularDampingOverride=0.99
    angularSpringConstant=15000
    bAngularSpring=true
    bOverrideAngularDamping=true
    bUseGravityOverride=true
    simulationSpace=RootRelative
    alphaBoolBlend {blendOption=Linear}
    boundBone {boneName=dyn_frontHair_main_01}
    componentAppliedLinearAccClamp=<10000 10000 10000>
    componentLinearAccScale=<1 1 1>
    componentLinearVelScale=<1 1 1>
    physicsBodyDefinitions [{
      boundBone {boneName=dyn_frontHair_main_01}
      boxExtents=<17 20 20>
      constraintSetup {angularLimitsMax=<3 5 5>, angularLimitsMin=<-3 -5 -5>, angularTarget=<1 0 0>}
      localJointOffset=<4 0 0>
     }]
   }
   n3 = Anim Dynamics - Bone: dyn_frontHair_main_02!(ComponentPose=@n2, Alpha=1, GravityOverride:Vector="0,0,0", ExternalForce="0,0,0") -> (Pose) {
    angularDampingOverride=0.99
    angularSpringConstant=15000
    bOverrideAngularDamping=true
    bUseGravityOverride=true
    simulationSpace=BoneRelative
    alphaBoolBlend {blendOption=Linear}
    boundBone {boneName=dyn_frontHair_main_02}
    componentAppliedLinearAccClamp=<10000 10000 10000>
    componentLinearAccScale=<1 1 1>
    componentLinearVelScale=<1 1 1>
    physicsBodyDefinitions [{
      boundBone {boneName=dyn_frontHair_main_02}
      boxExtents=<15 20 20>
      constraintSetup {angularLimitsMax=<0 10 10>, angularLimitsMin=<0 -10 -10>, angularTarget=<1 0 0>}
      localJointOffset=<4 0 0>
     }]
    relativeSpaceBone {boneName=spine_03}
   }
   Save cached pose 'CRCache_mainDynamics'(Pose=Component To Local(ComponentPose=@n3)) {cacheName=CRCache_mainDynamics}
   n4 = "Constraint - base_frontHair_03 (PPPPP)"!(
    ComponentPose=Local To Component(LocalPose=Use cached pose 'CRCache_mainDynamics'())
    Alpha=1
    ConstraintWeights_0=0.8033
    ConstraintWeights_1:float=0.1565
    ConstraintWeights_2:float=0.0361
    ConstraintWeights_3:float=0.0039
    ConstraintWeights_4:float=0.0002
   ) -> (Pose) {
    alphaBoolBlend {blendOption=Linear}
    boneToModify {boneName=base_frontHair_03}
    constraintSetup [
     {targetBone {boneName=head}}
     {targetBone {boneName=dyn_frontHair_03}}
     {targetBone {boneName=dyn_frontHair_01}}
     {targetBone {boneName=dyn_frontHair_main_01}}
     {targetBone {boneName=dyn_frontHair_02}}
    ]
    constraintWeights [1,1,1,1,1]
   }
   n5 = "Constraint - base_frontHair_02 (PPPP)"!(
    ComponentPose=@n4
    Alpha=1
    ConstraintWeights_0=0.6862
    ConstraintWeights_1:float=0.2712
    ConstraintWeights_2:float=0.0256
    ConstraintWeights_3:float=0.0169
   ) -> (Pose) {
    alphaBoolBlend {blendOption=Linear}
    boneToModify {boneName=base_frontHair_02}
    constraintSetup [
     {targetBone {boneName=dyn_frontHair_02}}
     {targetBone {boneName=head}}
     {targetBone {boneName=dyn_frontHair_main_01}}
     {targetBone {boneName=dyn_frontHair_01}}
    ]
    constraintWeights [1,1,1,1]
   }
   n6 = "Constraint - base_frontHair_01 (PPPPP)"!(
    ComponentPose=@n5
    Alpha=1
    ConstraintWeights_0=0.5922
    ConstraintWeights_1:float=0.3192
    ConstraintWeights_2:float=0.0818
    ConstraintWeights_3:float=0.0039
    ConstraintWeights_4:float=0.003
   ) -> (Pose) {
    alphaBoolBlend {blendOption=Linear}
    boneToModify {boneName=base_frontHair_01}
    constraintSetup [
     {targetBone {boneName=dyn_frontHair_01}}
     {targetBone {boneName=head}}
     {targetBone {boneName=dyn_frontHair_03}}
     {targetBone {boneName=dyn_frontHair_02}}
     {targetBone {boneName=dyn_frontHair_main_01}}
    ]
    constraintWeights [1,1,1,1,1]
   }
   Save cached pose 'CRCache_constrains'(Pose=Component To Local(ComponentPose=@n6)) {cacheName=CRCache_constrains}
   n7 = Anim Dynamics - Bone: dyn_frontHair_03!(
    ComponentPose=Local To Component(LocalPose=Use cached pose 'CRCache_constrains'())
    Alpha=1
    GravityOverride:Vector="0,0,-980"
    ExternalForce="0,0,0"
   ) -> (Pose) {
    angularDampingOverride=1
    angularSpringConstant=5000
    bAngularSpring=true
    bOverrideAngularDamping=true
    bUseGravityOverride=true
    gravityScale=0.35
    simulationSpace=BoneRelative
    alphaBoolBlend {blendOption=Linear}
    boundBone {boneName=dyn_frontHair_03}
    componentAppliedLinearAccClamp=<10000 10000 10000>
    componentLinearAccScale=<1 1 1>
    componentLinearVelScale=<1 1 1>
    gravityOverride {z=-980}
    physicsBodyDefinitions [{
      boundBone {boneName=dyn_frontHair_03}
      boxExtents=<15 20 20>
      constraintSetup {angularLimitsMax=<0 10 10>, angularLimitsMin=<0 -10 -10>, angularTarget=<1 0 0>}
      localJointOffset=<2 0 0>
     }]
    relativeSpaceBone {boneName=spine_03}
   }
   n8 = Anim Dynamics - Bone: dyn_frontHair_02!(ComponentPose=@n7, Alpha=1, GravityOverride:Vector="0,0,-980", ExternalForce="0,0,0") -> (Pose) {
    angularDampingOverride=1
    angularSpringConstant=5000
    bAngularSpring=true
    bOverrideAngularDamping=true
    bUseGravityOverride=true
    gravityScale=0.35
    simulationSpace=BoneRelative
    alphaBoolBlend {blendOption=Linear}
    boundBone {boneName=dyn_frontHair_02}
    componentAppliedLinearAccClamp=<10000 10000 10000>
    componentLinearAccScale=<1 1 1>
    componentLinearVelScale=<1 1 1>
    gravityOverride {z=-980}
    physicsBodyDefinitions [{
      boundBone {boneName=dyn_frontHair_02}
      boxExtents=<15 20 20>
      constraintSetup {angularLimitsMax=<0 10 10>, angularLimitsMin=<0 -10 -10>, angularTarget=<1 0 0>}
      localJointOffset=<2 0 0>
     }]
    relativeSpaceBone {boneName=spine_03}
   }
   n9 = Anim Dynamics - Bone: dyn_frontHair_01!(ComponentPose=@n8, Alpha=1, GravityOverride:Vector="0,0,-980", ExternalForce="0,0,0") -> (Pose) {
    angularDampingOverride=1
    angularSpringConstant=5000
    bAngularSpring=true
    bOverrideAngularDamping=true
    bUseGravityOverride=true
    gravityScale=0.35
    simulationSpace=BoneRelative
    alphaBoolBlend {blendOption=Linear}
    boundBone {boneName=dyn_frontHair_01}
    componentAppliedLinearAccClamp=<10000 10000 10000>
    componentLinearAccScale=<1 1 1>
    componentLinearVelScale=<1 1 1>
    gravityOverride {z=-980}
    physicsBodyDefinitions [{
      boundBone {boneName=dyn_frontHair_01}
      boxExtents=<15 20 20>
      constraintSetup {angularLimitsMax=<0 10 10>, angularLimitsMin=<0 -10 -10>, angularTarget=<1 0 0>}
      localJointOffset=<4 0 0>
     }]
    relativeSpaceBone {boneName=spine_03}
   }
   Save cached pose 'CRCache_secondaryDynamics'(Pose=Component To Local(ComponentPose=@n9)) {cacheName=CRCache_secondaryDynamics}
  ]}
 EventGraph {nodes [
   Event BlueprintUpdateAnimation()
   > n1 = Gravity Override State Function DEPRECATED(CharacterPartAnimInstance=Self-Reference(), GameTimeInSeconds=GetGameTimeInSeconds(), GravityOverrideParams=Get Gravity Override Params())
   > Set GravityOverride(GravityOverride=@n1)
   Event OnWeaponChanged()
   > Set PistolOut(PistolOut:bool=false) -> (Output_Get:bool)
   > n2 = Switch on EFortWeaponCoreAnimation!(execute, Selection:EFortWeaponCoreAnimation Enum=Get WeaponCoreAnim!() -> (WeaponCoreAnim)) -> (
    Melee:Exec
    Pistol:Exec
    Shotgun:Exec
    PaperBlueprint:Exec
    Rifle:Exec
    MeleeOneHand:Exec
    MachinePistol:Exec
    RocketLauncher:Exec
    GrenadeLauncher:Exec
    GoingCommando:Exec
    AssaultRifle:Exec
    TacticalShotgun:Exec
    SniperRifle:Exec
    TrapPlacement:Exec
    ShoulderLauncher:Exec
    AbilityDecoTool:Exec
    Crossbow:Exec
    C4:Exec
    RemoteControl:Exec
    DualWield:Exec
    AR_BullPup:Exec
    AR_ForwardGrip:Exec
    MedPackPaddles:Exec
    SMG_P90:Exec
    AR_DrumGun:Exec
    Consumable_Small:Exec
    Consumable_Large:Exec
    Balloon:Exec
    MountedTurret:Exec
    CreativeTool:Exec
    ExplosiveBow:Exec
    AshtonIndigo:Exec
    AshtonChicago:Exec
    MeleeDualWield:Exec
    Unarmed:Exec
   )
   Set PistolOut(execute=[@n2.DualWield, @n2.Pistol], PistolOut=true)
  ]}
]
variables [
 Gravity Override Params {type=GravityOverrideParamsStruct}
 GravityOverride {type=Vector}
 PistolOut {type=boolean}
]
